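# Physics asset for the PinkSpike backpack cosmetic, covering the right ear and the head it hangs from.
# Capsule bodies form one constrained chain from pack through base_head and dyn_head to three right-ear segments.
# Swing limits are tight near the pack and ear root and looser along the outer ear joints, while twist is nearly locked throughout, so the ear bends rather than rolls.
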PhysicsAsset /BRCosmetics/Accessories/FORT_Backpacks/Backpack_F_MED_PinkSpike/Meshes/F_MED_PinkSpike_Pack_Ear_R_Physics1
bodies {
 base_head [{shapeType=Capsule, length=1, radius=1, shapeName=base_head_capsule, rotation=<rot 90 0 0>}]
 dyn_R_ear01 [{shapeType=Capsule, length=4, radius=1, shapeName=dyn_R_ear01_capsule, center=<3.4047 0 0>, rotation=<rot 90 0 0>}]
 dyn_R_ear02 [{shapeType=Capsule, length=4, radius=1, shapeName=dyn_R_ear02_capsule, center=<3.4047 0 0>, rotation=<rot 90 0 0>}]
 dyn_R_ear03 [{shapeType=Capsule, length=4, radius=1, shapeName=dyn_R_ear03_capsule, center=<3.4047 0 0>, rotation=<rot 90 0 0>}]
 dyn_head [{shapeType=Capsule, length=2, radius=5, shapeName=dyn_head_capsule, center=<6.602 0 0>, rotation=<rot 90 0 0>}]
 pack [{shapeType=Capsule, length=1, radius=2, shapeName=pack_capsule, center=<0 -2.9144 -1.9966>, rotation=<rot 90 0 0>}]
}
constraints [
 {bone1Name=base_head, bone2Name=pack, swing1LimitDegrees=5, swing2LimitDegrees=5, twistLimitDegrees=1}
 {bone1Name=dyn_head, bone2Name=base_head, swing1LimitDegrees=12, swing2LimitDegrees=15, twistLimitDegrees=1}
 {bone1Name=dyn_R_ear01, bone2Name=dyn_head, swing1LimitDegrees=10, swing2LimitDegrees=10, twistLimitDegrees=1}
 {bone1Name=dyn_R_ear02, bone2Name=dyn_R_ear01, swing1LimitDegrees=35, swing2LimitDegrees=25, twistLimitDegrees=1}
 {bone1Name=dyn_R_ear03, bone2Name=dyn_R_ear02, swing1LimitDegrees=25, swing2LimitDegrees=30, twistLimitDegrees=1}
]
